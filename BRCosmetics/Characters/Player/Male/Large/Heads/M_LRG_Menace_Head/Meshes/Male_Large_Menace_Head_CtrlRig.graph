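# Control Rig for the Male Large Menace head that procedurally corrects head rotation based on the current pose.
# It reads bone-relative rotations as Euler angles, remaps them through clamped ranges and feeds the results into additive Modify Transforms steps.
# In the executed chain, only the first head correction, driven by head-to-neck_02 Z rotation, has weight; later head and neck_01 steps are zero-weighted.
# Several variant chains and stray remap nodes remain disconnected from execution.
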ControlRigBlueprint /BRCosmetics/Characters/Player/Male/Large/Heads/M_LRG_Menace_Head/Meshes/Male_Large_Menace_Head_CtrlRig
graphs [
 RigVMFunctionLibrary {links [], nodes []}
 RigVMModel {nodes [
   n1 = &RigVMFunction_MathFloatRemap(
    SourceMaximum=-7
    SourceMinimum=-2
    TargetMaximum=100
    TargetMinimum=0
    bClamp=true
    Value=&RigVMFunction_MathQuaternionToEuler(
     RotationOrder=ZYX
     Value=&RigUnit_GetRelativeTransformForItem(Child=(Type=Bone,Name="neck_02"), Parent=(Type=Bone,Name="spine_05")).RelativeTransform.Rotation
    ).Result.X
   )
   &RigUnit_ModifyTransforms(
    ExecutePin=()
    ItemToModify=((Item=(Type=Bone,Name="neck_01"),Transform=(Rotation=(X=0,Y=0,Z=0,W=1),Translation=(X=0,Y=0,Z=0),Scale3D=(X=1,Y=1,Z=1))))
    Weight=0
    WeightMinimum=0
    ItemToModify.0.Transform.Rotation=&RigVMFunction_MathQuaternionFromEuler(Euler=(X=0,Y=0,Z=0), Euler.X=@n1.Result, Euler.Y=@n1.Result).Result
   )
   n2 = &RigVMFunction_MathQuaternionToEuler(
    RotationOrder=ZYX
    Value=&RigUnit_GetRelativeTransformForItem(Child=(Type=Bone,Name="head"), Parent=(Type=Bone,Name="neck_02")).RelativeTransform.Rotation
   )
   > &RigUnit_ModifyTransforms(
    ItemToModify=((Item=(Type=Bone,Name="head"),Transform=(Rotation=(X=0,Y=0,Z=0,W=1),Translation=(X=0,Y=0,Z=0),Scale3D=(X=1,Y=1,Z=1))))
    Weight=1
    ItemToModify.0.Transform.Rotation=&RigVMFunction_MathQuaternionFromEuler(
     Euler=(X=0,Y=0,Z=0)
     Euler.Z=&RigVMFunction_MathFloatRemap(SourceMaximum=-29, SourceMinimum=-20, TargetMaximum=16, TargetMinimum=0, bClamp=true, Value=@n2.Result.Z).Result
    ).Result
   )
   RigUnit_BeginExecution = &RigUnit_BeginExecution()
   n3 = &RigVMFunction_MathQuaternionToEuler(
    RotationOrder=ZYX
    Value=&RigUnit_GetRelativeTransformForItem(Child=(Type=Bone,Name="head"), Parent=(Type=Bone,Name="neck_02")).RelativeTransform.Rotation
   )
   RigVMStructNode_0_3 = &RigVMFunction_MathQuaternionFromEuler(
    Euler=(X=0,Y=0,Z=0)
    Euler.Z=&RigVMFunction_MathFloatRemap(SourceMaximum=-50, SourceMinimum=-17, TargetMaximum=27, TargetMinimum=0, bClamp=true, Value=@n3.Result.Z).Result
   )
   RigVMStructNode_1_3 = &RigUnit_ModifyTransforms(
    ItemToModify=((Item=(Type=Bone,Name="head"),Transform=(Rotation=(X=0,Y=0,Z=0,W=1),Translation=(X=0,Y=0,Z=0),Scale3D=(X=1,Y=1,Z=1))))
    Weight=1
    ExecutePin=@RigUnit_BeginExecution.ExecutePin
    ItemToModify.0.Transform.Rotation=@RigVMStructNode_0_3.Result
   )
   > &RigUnit_ModifyTransforms(
    ItemToModify=((Item=(Type=Bone,Name="head"),Transform=(Rotation=(X=0,Y=0,Z=0,W=1),Translation=(X=0,Y=0,Z=0),Scale3D=(X=1,Y=1,Z=1))))
    Weight=0
    ItemToModify.0.Transform.Rotation=&RigVMFunction_MathQuaternionFromEuler(
     Euler=(X=0,Y=0,Z=0)
     Euler.Z=&RigVMFunction_MathFloatRemap(SourceMaximum=7, SourceMinimum=-11, TargetMaximum=8, TargetMinimum=0, bClamp=true, Value=@n3.Result.Z).Result
    ).Result
   )
   n4 = &RigVMFunction_MathQuaternionToEuler(
    RotationOrder=ZYX
    Value=&RigUnit_GetRelativeTransformForItem(Child=(Type=Bone,Name="neck_01"), Parent=(Type=Bone,Name="spine_05")).RelativeTransform.Rotation
   )
   RigVMStructNode_0_1_2 = &RigVMFunction_MathQuaternionFromEuler(
    Euler=(X=0,Y=0,Z=0)
    Euler.Z=&RigVMFunction_MathFloatRemap(SourceMaximum=2, SourceMinimum=14, TargetMaximum=6, TargetMinimum=0, bClamp=true, Value=@n4.Result.Z).Result
   )
   > RigVMStructNode_1_1_2 = &RigUnit_ModifyTransforms(
    ItemToModify=((Item=(Type=Bone,Name="neck_01"),Transform=(Rotation=(X=0,Y=0,Z=0,W=1),Translation=(X=0,Y=0,Z=0),Scale3D=(X=1,Y=1,Z=1))))
    Weight=0
    ItemToModify.0.Transform.Rotation=@RigVMStructNode_0_1_2.Result
   )
   > &RigUnit_ModifyTransforms(
    ItemToModify=((Item=(Type=Bone,Name="neck_01"),Transform=(Rotation=(X=0,Y=0,Z=0,W=1),Translation=(X=0,Y=0,Z=0),Scale3D=(X=1,Y=1,Z=1))))
    Weight=0
    WeightMinimum=0
    ItemToModify.0.Transform.Rotation=&RigVMFunction_MathQuaternionFromEuler(
     Euler=(X=0,Y=0,Z=0)
     Euler.Z=&RigVMFunction_MathFloatRemap(SourceMaximum=32, SourceMinimum=20, TargetMaximum=-16, TargetMinimum=0, bClamp=true, Value=@n4.Result.Z).Result
    ).Result
   )
   n5 = &RigVMFunction_MathQuaternionToEuler(
    RotationOrder=ZYX
    Value=&RigUnit_GetRelativeTransformForItem(Child=(Type=Bone,Name="head"), Parent=(Type=Bone,Name="neck_01")).RelativeTransform.Rotation
   )
   RigVMStructNode_0 = &RigVMFunction_MathQuaternionFromEuler(
    Euler=(X=0,Y=0,Z=0)
    Euler.Z=&RigVMFunction_MathFloatRemap(SourceMaximum=-50, SourceMinimum=-17, TargetMaximum=27, TargetMinimum=0, bClamp=true, Value=@n5.Result.Z).Result
   )
   RigVMStructNode_1 = &RigUnit_ModifyTransforms(
    ExecutePin=()
    ItemToModify=((Item=(Type=Bone,Name="head"),Transform=(Rotation=(X=0,Y=0,Z=0,W=1),Translation=(X=0,Y=0,Z=0),Scale3D=(X=1,Y=1,Z=1))))
    Weight=1
    ItemToModify.0.Transform.Rotation=@RigVMStructNode_0.Result
   )
   > &RigUnit_ModifyTransforms(
    ItemToModify=((Item=(Type=Bone,Name="head"),Transform=(Rotation=(X=0,Y=0,Z=0,W=1),Translation=(X=0,Y=0,Z=0),Scale3D=(X=1,Y=1,Z=1))))
    Weight=1
    ItemToModify.0.Transform.Rotation=&RigVMFunction_MathQuaternionFromEuler(
     Euler=(X=0,Y=0,Z=0)
     Euler.Z=&RigVMFunction_MathFloatRemap(SourceMaximum=0, SourceMinimum=-17, TargetMaximum=0, TargetMinimum=0, bClamp=true, Value=@n5.Result.Z).Result
    ).Result
   )
   n6 = &RigVMFunction_MathQuaternionToEuler(
    RotationOrder=ZYX
    Value=&RigUnit_GetRelativeTransformForItem(Child=(Type=Bone,Name="neck_01"), Parent=(Type=Bone,Name="spine_05")).RelativeTransform.Rotation
   )
   RigVMStructNode_0_1 = &RigVMFunction_MathQuaternionFromEuler(
    Euler=(X=0,Y=0,Z=0)
    Euler.Z=&RigVMFunction_MathFloatRemap(SourceMaximum=0, SourceMinimum=-17, TargetMaximum=0, TargetMinimum=0, bClamp=true, Value=@n6.Result.Z).Result
   )
   > RigVMStructNode_1_1 = &RigUnit_ModifyTransforms(
    ItemToModify=((Item=(Type=Bone,Name="neck_01"),Transform=(Rotation=(X=0,Y=0,Z=0,W=1),Translation=(X=0,Y=0,Z=0),Scale3D=(X=1,Y=1,Z=1))))
    Weight=0
    ItemToModify.0.Transform.Rotation=@RigVMStructNode_0_1.Result
   )
   > &RigUnit_ModifyTransforms(
    ItemToModify=((Item=(Type=Bone,Name="neck_01"),Transform=(Rotation=(X=0,Y=0,Z=0,W=1),Translation=(X=0,Y=0,Z=0),Scale3D=(X=1,Y=1,Z=1))))
    Weight=1
    WeightMinimum=0
    ItemToModify.0.Transform.Rotation=&RigVMFunction_MathQuaternionFromEuler(
     Euler=(X=0,Y=0,Z=0)
     Euler.Z=&RigVMFunction_MathFloatRemap(SourceMaximum=32, SourceMinimum=20, TargetMaximum=-13, TargetMinimum=0, bClamp=true, Value=@n6.Result.Z).Result
    ).Result
   )
   n7 = &RigVMFunction_MathQuaternionToEuler(
    RotationOrder=ZYX
    Value=&RigUnit_GetRelativeTransformForItem(Child=(Type=Bone,Name="head"), Parent=(Type=Bone,Name="neck_02")).RelativeTransform.Rotation
   )
   RigVMStructNode_0_3_1 = &RigVMFunction_MathQuaternionFromEuler(
    Euler=(X=0,Y=0,Z=0)
    Euler.Z=&RigVMFunction_MathFloatRemap(SourceMaximum=-50, SourceMinimum=-17, TargetMaximum=27, TargetMinimum=0, bClamp=true, Value=@n7.Result.Z).Result
   )
   RigVMStructNode_1_3_1 = &RigUnit_ModifyTransforms(
    ExecutePin=()
    ItemToModify=((Item=(Type=Bone,Name="head"),Transform=(Rotation=(X=0,Y=0,Z=0,W=1),Translation=(X=0,Y=0,Z=0),Scale3D=(X=1,Y=1,Z=1))))
    Weight=1
    ItemToModify.0.Transform.Rotation=@RigVMStructNode_0_3_1.Result
   )
   > &RigUnit_ModifyTransforms(
    ItemToModify=((Item=(Type=Bone,Name="head"),Transform=(Rotation=(X=0,Y=0,Z=0,W=1),Translation=(X=0,Y=0,Z=0),Scale3D=(X=1,Y=1,Z=1))))
    Weight=1
    ItemToModify.0.Transform.Rotation=&RigVMFunction_MathQuaternionFromEuler(
     Euler=(X=0,Y=0,Z=0)
     Euler.Z=&RigVMFunction_MathFloatRemap(SourceMaximum=7, SourceMinimum=-11, TargetMaximum=8, TargetMinimum=0, bClamp=true, Value=@n7.Result.Z).Result
    ).Result
   )
   n8 = &RigVMFunction_MathQuaternionToEuler(
    RotationOrder=ZYX
    Value=&RigUnit_GetRelativeTransformForItem(Child=(Type=Bone,Name="neck_01"), Parent=(Type=Bone,Name="spine_05")).RelativeTransform.Rotation
   )
   RigVMStructNode_0_1_2_1 = &RigVMFunction_MathQuaternionFromEuler(
    Euler=(X=0,Y=0,Z=0)
    Euler.Z=&RigVMFunction_MathFloatRemap(SourceMaximum=2, SourceMinimum=14, TargetMaximum=6, TargetMinimum=0, bClamp=true, Value=@n8.Result.Z).Result
   )
   > RigVMStructNode_1_1_2_1 = &RigUnit_ModifyTransforms(
    ItemToModify=((Item=(Type=Bone,Name="neck_01"),Transform=(Rotation=(X=0,Y=0,Z=0,W=1),Translation=(X=0,Y=0,Z=0),Scale3D=(X=1,Y=1,Z=1))))
    Weight=1
    ItemToModify.0.Transform.Rotation=@RigVMStructNode_0_1_2_1.Result
   )
   > &RigUnit_ModifyTransforms(
    ItemToModify=((Item=(Type=Bone,Name="neck_01"),Transform=(Rotation=(X=0,Y=0,Z=0,W=1),Translation=(X=0,Y=0,Z=0),Scale3D=(X=1,Y=1,Z=1))))
    Weight=1
    WeightMinimum=0
    ItemToModify.0.Transform.Rotation=&RigVMFunction_MathQuaternionFromEuler(
     Euler=(X=0,Y=0,Z=0)
     Euler.Z=&RigVMFunction_MathFloatRemap(SourceMaximum=32, SourceMinimum=20, TargetMaximum=-16, TargetMinimum=0, bClamp=true, Value=@n8.Result.Z).Result
    ).Result
   )
   &RigVMFunction_MathFloatRemap(SourceMaximum=4, SourceMinimum=-35, TargetMinimum=24, Value=0, bClamp=true)
   &RigVMFunction_MathFloatRemap(SourceMaximum=4, SourceMinimum=-35, TargetMinimum=24, Value=0, bClamp=true)
   &RigVMFunction_MathFloatRemap(SourceMaximum=27, SourceMinimum=-6.2, TargetMaximum=5, TargetMinimum=-4, Value=0, bClamp=true)
   &RigVMFunction_MathFloatRemap(SourceMaximum=-29, SourceMinimum=-20, TargetMaximum=16, TargetMinimum=0, Value=0, bClamp=true)
  ]}
]
hierarchy [
 root {loc=<0 0 0>, rot=<0 0 0 1>}
 pelvis {parent=root, loc=<0 -0.7321 92.4218>, rot=<0.0127 0.707 -0.0127 -0.707>}
 spine_01 {parent=pelvis, loc=<5.6002 -0.5128 0>, rot=<0 0 0.034 -0.9994>}
 spine_02 {parent=spine_01, loc=<8.0753 0.0139 0>, rot=<0 0 0.0238 -0.9997>}
 spine_03 {parent=spine_02, loc=<7.0393 0 0>, rot=<0 0 0.0149 -0.9999>}
 spine_04 {parent=spine_03, loc=<8.1024 0 0>, rot=<0 0 -0.0132 -0.9999>}
 spine_05 {parent=spine_04, loc=<10.0146 0 0>, rot=<0 0 -0.0015 -1>}
 clavicle_l {parent=spine_05, loc=<2.5856 -0.2866 -4.3777>, rot=<0.0465 -0.7289 -0.009 -0.6829>}
 clavicle_r {parent=spine_05, loc=<2.5856 -0.2865 4.3777>, rot=<0.7289 0.0465 -0.6829 0.009>}
 neck_01 {parent=spine_05, loc=<10.815 1.9196 0>, rot=<0 0 -0.1864 -0.9825>}
 neck_02 {parent=neck_01, loc=<4.7347 0.3045 0>, rot=<0 0 0.0363 -0.9993>}
 head {parent=neck_02, loc=<5.7632 0.5552 0>, rot=<0 0 0.115 -0.9934>}
 faceAttach {parent=head, loc=<-0.1673 -2.8491 -0.0701>, rot=<0.0082 -0.7071 -0.0082 -0.7071>, scale=<0.89 0.89 0.89>}
 C_brow_mid {parent=faceAttach, loc=<-0.1054 12.0239 6.372>, rot=<0 0.0891 -0.0176 -0.9959>}
 L_brow_mid {parent=faceAttach, loc=<2.9144 11.0921 6.8645>, rot=<-0.0002 -0.0051 0.2553 -0.9669>}
 L_brow_outer {parent=faceAttach, loc=<5.2499 8.874 7.0603>, rot=<0.0105 0.0093 0.4403 -0.8977>}
 L_cheek_inner {parent=faceAttach, loc=<4.1165 10.4261 1.4124>, rot=<-0.1218 -0.3977 0.2963 -0.8597>}
 C_lip_upper_mid {parent=faceAttach, loc=<-0.0652 13.0289 -1.0882>, rot=<0.2225 0.0017 -0.0006 -0.9749>}
 L_lip_upper_outer {parent=faceAttach, loc=<1.9847 12.249 -0.9219>, rot=<0.0379 -0.0674 0.2607 -0.9623>}
 L_lip_corner {parent=faceAttach, loc=<3.1561 10.966 -1.3419>, rot=<0.0019 0.0091 -0.0145 -0.9999>}
 C_nose_bridge {parent=faceAttach, loc=<-0.086 13.1815 4.3112>, rot=<-0.2218 -0.0032 -0.0009 -0.9751>}
 R_brow_mid {parent=faceAttach, loc=<-3.0712 11.0936 6.8699>, rot=<-0.0001 0.0043 -0.2553 -0.9669>}
 R_brow_outer {parent=faceAttach, loc=<-5.4226 8.8563 7.0514>, rot=<-0.0029 0.0075 -0.4436 -0.8962>}
 R_cheek_inner {parent=faceAttach, loc=<-4.2651 10.4376 1.4052>, rot=<-0.1208 0.395 -0.2969 -0.8609>}
 R_lip_corner {parent=faceAttach, loc=<-3.3024 11.0329 -1.2898>, rot=<0.0027 -0.0083 0.0146 -0.9999>}
 C_jaw {parent=faceAttach, loc=<-0.1354 3.293 0.342>, rot=<0 0 0 1>}
 L_lip_lower_outer {parent=C_jaw, loc=<2.2138 8.7793 -1.2825>, rot=<0.0279 0.1255 0.3654 -0.9219>}
 C_lip_lower_mid {parent=C_jaw, loc=<0.0505 10.1226 -1.3723>, rot=<0 0.0008 0.0002 -1>}
 R_lip_lower_outer {parent=C_jaw, loc=<-2.1156 8.9142 -1.2746>, rot=<0.0213 -0.0897 -0.3394 -0.9361>}
 teeth_lower {parent=C_jaw, loc=<-0.063 7.62 -3.2335>, rot=<0 0 0 1>}
 tongue {parent=C_jaw, loc=<0.0403 1.9996 -0.5921>, rot=<-0.1568 0 0 -0.9876>}
 R_eye {parent=faceAttach, loc=<-2.9502 9.2689 5.8498>, rot=<0 0 0.9998 -0.0187>}
 R_eye_lid_upper_mid {parent=R_eye, loc=<-0.0396 -1.1712 0.2636>, rot=<0.0523 -0.0009 0.0167 -0.9985>}
 R_eye_lid_lower_mid {parent=R_eye, loc=<-0.0085 -0.9422 -0.7046>, rot=<0.0329 -0.0005 0.0167 -0.9993>}
 L_eye {parent=faceAttach, loc=<2.7927 9.2689 5.8498>, rot=<0 0 0.9998 0.0187>}
 L_eye_lid_upper_mid {parent=L_eye, loc=<0.0251 -1.1623 0.2388>, rot=<0.0523 0 0 -0.9986>}
 L_eye_lid_lower_mid {parent=L_eye, loc=<0.0223 -0.923 -0.7488>, rot=<0.0329 0 0 -0.9995>}
 teeth_upper {parent=faceAttach, loc=<-0.0785 11.3312 0.4977>, rot=<0 0 0 1>}
 R_lip_upper_outer {parent=faceAttach, loc=<-2.166 12.3084 -0.9168>, rot=<0.044 0.0499 -0.254 -0.9649>}
 look_left_pose {type=Curve, loc=<0 0 0>, rot=<0 0 0 1>}
 look_right_pose {type=Curve, loc=<0 0 0>, rot=<0 0 0 1>}
 look_up_pose {type=Curve, loc=<0 0 0>, rot=<0 0 0 1>}
 look_down_pose {type=Curve, loc=<0 0 0>, rot=<0 0 0 1>}
 L_blink_pose {type=Curve, loc=<0 0 0>, rot=<0 0 0 1>}
 R_blink_pose {type=Curve, loc=<0 0 0>, rot=<0 0 0 1>}
 L_wide_pose {type=Curve, loc=<0 0 0>, rot=<0 0 0 1>}
 R_wide_pose {type=Curve, loc=<0 0 0>, rot=<0 0 0 1>}
 R_squint_inner_pose {type=Curve, loc=<0 0 0>, rot=<0 0 0 1>}
 R_squeeze_pose {type=Curve, loc=<0 0 0>, rot=<0 0 0 1>}
 L_squint_inner_pose {type=Curve, loc=<0 0 0>, rot=<0 0 0 1>}
 L_squeeze_pose {type=Curve, loc=<0 0 0>, rot=<0 0 0 1>}
 L_brow_up_pose {type=Curve, loc=<0 0 0>, rot=<0 0 0 1>}
 L_brow_down_pose {type=Curve, loc=<0 0 0>, rot=<0 0 0 1>}
 R_brow_up_pose {type=Curve, loc=<0 0 0>, rot=<0 0 0 1>}
 R_brow_down_pose {type=Curve, loc=<0 0 0>, rot=<0 0 0 1>}
 C_glabella_up_pose {type=Curve, loc=<0 0 0>, rot=<0 0 0 1>}
 C_glabella_down_pose {type=Curve, loc=<0 0 0>, rot=<0 0 0 1>}
 L_brow_squeeze_in_pose {type=Curve, loc=<0 0 0>, rot=<0 0 0 1>}
 R_brow_squeeze_in_pose {type=Curve, loc=<0 0 0>, rot=<0 0 0 1>}
 L_upper_lip_raiser_pose {type=Curve, loc=<0 0 0>, rot=<0 0 0 1>}
 R_upper_lip_raiser_pose {type=Curve, loc=<0 0 0>, rot=<0 0 0 1>}
 L_upper_lip_lower_pose {type=Curve, loc=<0 0 0>, rot=<0 0 0 1>}
 R_upper_lip_lower_pose {type=Curve, loc=<0 0 0>, rot=<0 0 0 1>}
 L_lower_lip_up_pose {type=Curve, loc=<0 0 0>, rot=<0 0 0 1>}
 R_lower_lip_up_pose {type=Curve, loc=<0 0 0>, rot=<0 0 0 1>}
 L_lower_lip_down_pose {type=Curve, loc=<0 0 0>, rot=<0 0 0 1>}
 R_lower_lip_down_pose {type=Curve, loc=<0 0 0>, rot=<0 0 0 1>}
 L_lip_corner_narrow_pose {type=Curve, loc=<0 0 0>, rot=<0 0 0 1>}
 L_lip_corner_wide_pose {type=Curve, loc=<0 0 0>, rot=<0 0 0 1>}
 R_lip_corner_narrow_pose {type=Curve, loc=<0 0 0>, rot=<0 0 0 1>}
 R_lip_corner_wide_pose {type=Curve, loc=<0 0 0>, rot=<0 0 0 1>}
 L_smile_pose {type=Curve, loc=<0 0 0>, rot=<0 0 0 1>}
 R_smile_pose {type=Curve, loc=<0 0 0>, rot=<0 0 0 1>}
 L_frown_pose {type=Curve, loc=<0 0 0>, rot=<0 0 0 1>}
 R_frown_pose {type=Curve, loc=<0 0 0>, rot=<0 0 0 1>}
 phoneme_oo_pose {type=Curve, loc=<0 0 0>, rot=<0 0 0 1>}
 phoneme_ch_pose {type=Curve, loc=<0 0 0>, rot=<0 0 0 1>}
 phoneme_mbp_pose {type=Curve, loc=<0 0 0>, rot=<0 0 0 1>}
 phoneme_fv_pose {type=Curve, loc=<0 0 0>, rot=<0 0 0 1>}
 mouth_up_pose {type=Curve, loc=<0 0 0>, rot=<0 0 0 1>}
 mouth_down_pose {type=Curve, loc=<0 0 0>, rot=<0 0 0 1>}
 C_nose_wrinkler_pose {type=Curve, loc=<0 0 0>, rot=<0 0 0 1>}
 L_cheek_up_pose {type=Curve, loc=<0 0 0>, rot=<0 0 0 1>}
 R_cheek_up_pose {type=Curve, loc=<0 0 0>, rot=<0 0 0 1>}
 jaw_open_pose {type=Curve, loc=<0 0 0>, rot=<0 0 0 1>}
 jaw_right_pose {type=Curve, loc=<0 0 0>, rot=<0 0 0 1>}
 jaw_left_pose {type=Curve, loc=<0 0 0>, rot=<0 0 0 1>}
 eye_convergence_pose {type=Curve, loc=<0 0 0>, rot=<0 0 0 1>}
 L_upperLidDown_pose {type=Curve, loc=<0 0 0>, rot=<0 0 0 1>}
 R_upperLidDown_pose {type=Curve, loc=<0 0 0>, rot=<0 0 0 1>}
 jaw_up_pose {type=Curve, loc=<0 0 0>, rot=<0 0 0 1>}
 jaw_forward_pose {type=Curve, loc=<0 0 0>, rot=<0 0 0 1>}
 jaw_back_pose {type=Curve, loc=<0 0 0>, rot=<0 0 0 1>}
 tongue_forward_pose {type=Curve, loc=<0 0 0>, rot=<0 0 0 1>}
 tongue_back_pose {type=Curve, loc=<0 0 0>, rot=<0 0 0 1>}
 tongue_up_pose {type=Curve, loc=<0 0 0>, rot=<0 0 0 1>}
 tongue_down_pose {type=Curve, loc=<0 0 0>, rot=<0 0 0 1>}
 base_pose {type=Curve, loc=<0 0 0>, rot=<0 0 0 1>}
 Hair_Scarf {type=Curve, loc=<0 0 0>, rot=<0 0 0 1>}
 Hair_Hat {type=Curve, loc=<0 0 0>, rot=<0 0 0 1>}
 Hair_Helmet {type=Curve, loc=<0 0 0>, rot=<0 0 0 1>}
 Hair_Glasses {type=Curve, loc=<0 0 0>, rot=<0 0 0 1>}
 Hair_Cap {type=Curve, loc=<0 0 0>, rot=<0 0 0 1>}
 Hair_Headsets {type=Curve, loc=<0 0 0>, rot=<0 0 0 1>}
 Hair_Hood {type=Curve, loc=<0 0 0>, rot=<0 0 0 1>}
 Face_Mask {type=Curve, loc=<0 0 0>, rot=<0 0 0 1>}
 HairStrands_Hat {type=Curve, loc=<0 0 0>, rot=<0 0 0 1>}
 Hat_Hair {type=Curve, loc=<0 0 0>, rot=<0 0 0 1>}
 Helmet_Hair {type=Curve, loc=<0 0 0>, rot=<0 0 0 1>}
 Hair_Muff {type=Curve, loc=<0 0 0>, rot=<0 0 0 1>}
 Hair_SpaceSuit {type=Curve, loc=<0 0 0>, rot=<0 0 0 1>}
 Hair_Hide {type=Curve, loc=<0 0 0>, rot=<0 0 0 1>}
 Head_SpaceSuit {type=Curve, loc=<0 0 0>, rot=<0 0 0 1>}
 Hair_Null {type=Curve, loc=<0 0 0>, rot=<0 0 0 1>}
 DisableFaceOverride {type=Curve, loc=<0 0 0>, rot=<0 0 0 1>}
 LiveLink {type=Curve, loc=<0 0 0>, rot=<0 0 0 1>}
 Hat_MESH {type=Curve, loc=<0 0 0>, rot=<0 0 0 1>}
 Null_MESH {type=Curve, loc=<0 0 0>, rot=<0 0 0 1>}
 Pose_0 {type=Curve, loc=<0 0 0>, rot=<0 0 0 1>}
 Hair_Mask {type=Curve, loc=<0 0 0>, rot=<0 0 0 1>}
 eye_convergence {type=Curve, loc=<0 0 0>, rot=<0 0 0 1>}
 headTransform {type=Curve, loc=<0 0 0>, rot=<0 0 0 1>}
 Pose_1 {type=Curve, loc=<0 0 0>, rot=<0 0 0 1>}
 Pose_2 {type=Curve, loc=<0 0 0>, rot=<0 0 0 1>}
 Pose_3 {type=Curve, loc=<0 0 0>, rot=<0 0 0 1>}
 Pose_4 {type=Curve, loc=<0 0 0>, rot=<0 0 0 1>}
 Pose_5 {type=Curve, loc=<0 0 0>, rot=<0 0 0 1>}
 Pose_6 {type=Curve, loc=<0 0 0>, rot=<0 0 0 1>}
 Pose_7 {type=Curve, loc=<0 0 0>, rot=<0 0 0 1>}
 Pose_8 {type=Curve, loc=<0 0 0>, rot=<0 0 0 1>}
 Pose_9 {type=Curve, loc=<0 0 0>, rot=<0 0 0 1>}
 Pose_10 {type=Curve, loc=<0 0 0>, rot=<0 0 0 1>}
 Pose_11 {type=Curve, loc=<0 0 0>, rot=<0 0 0 1>}
 Pose_12 {type=Curve, loc=<0 0 0>, rot=<0 0 0 1>}
 Pose_13 {type=Curve, loc=<0 0 0>, rot=<0 0 0 1>}
 Pose_14 {type=Curve, loc=<0 0 0>, rot=<0 0 0 1>}
 Pose_15 {type=Curve, loc=<0 0 0>, rot=<0 0 0 1>}
 Pose_16 {type=Curve, loc=<0 0 0>, rot=<0 0 0 1>}
 Pose_17 {type=Curve, loc=<0 0 0>, rot=<0 0 0 1>}
 Pose_18 {type=Curve, loc=<0 0 0>, rot=<0 0 0 1>}
 Pose_19 {type=Curve, loc=<0 0 0>, rot=<0 0 0 1>}
 Pose_20 {type=Curve, loc=<0 0 0>, rot=<0 0 0 1>}
 Pose_21 {type=Curve, loc=<0 0 0>, rot=<0 0 0 1>}
 Pose_22 {type=Curve, loc=<0 0 0>, rot=<0 0 0 1>}
 Pose_23 {type=Curve, loc=<0 0 0>, rot=<0 0 0 1>}
 Pose_24 {type=Curve, loc=<0 0 0>, rot=<0 0 0 1>}
 Pose_25 {type=Curve, loc=<0 0 0>, rot=<0 0 0 1>}
 Pose_26 {type=Curve, loc=<0 0 0>, rot=<0 0 0 1>}
 Pose_27 {type=Curve, loc=<0 0 0>, rot=<0 0 0 1>}
 Pose_28 {type=Curve, loc=<0 0 0>, rot=<0 0 0 1>}
 Pose_29 {type=Curve, loc=<0 0 0>, rot=<0 0 0 1>}
 Pose_30 {type=Curve, loc=<0 0 0>, rot=<0 0 0 1>}
 Pose_31 {type=Curve, loc=<0 0 0>, rot=<0 0 0 1>}
 Pose_32 {type=Curve, loc=<0 0 0>, rot=<0 0 0 1>}
 Pose_33 {type=Curve, loc=<0 0 0>, rot=<0 0 0 1>}
 Pose_34 {type=Curve, loc=<0 0 0>, rot=<0 0 0 1>}
 Pose_35 {type=Curve, loc=<0 0 0>, rot=<0 0 0 1>}
 Pose_36 {type=Curve, loc=<0 0 0>, rot=<0 0 0 1>}
 Pose_37 {type=Curve, loc=<0 0 0>, rot=<0 0 0 1>}
 Pose_38 {type=Curve, loc=<0 0 0>, rot=<0 0 0 1>}
 Pose_39 {type=Curve, loc=<0 0 0>, rot=<0 0 0 1>}
 Pose_40 {type=Curve, loc=<0 0 0>, rot=<0 0 0 1>}
 Pose_41 {type=Curve, loc=<0 0 0>, rot=<0 0 0 1>}
 Pose_42 {type=Curve, loc=<0 0 0>, rot=<0 0 0 1>}
 Pose_43 {type=Curve, loc=<0 0 0>, rot=<0 0 0 1>}
 Pose_44 {type=Curve, loc=<0 0 0>, rot=<0 0 0 1>}
 Pose_45 {type=Curve, loc=<0 0 0>, rot=<0 0 0 1>}
 Pose_46 {type=Curve, loc=<0 0 0>, rot=<0 0 0 1>}
 Pose_47 {type=Curve, loc=<0 0 0>, rot=<0 0 0 1>}
 Pose_48 {type=Curve, loc=<0 0 0>, rot=<0 0 0 1>}
 Pose_49 {type=Curve, loc=<0 0 0>, rot=<0 0 0 1>}
 Pose_50 {type=Curve, loc=<0 0 0>, rot=<0 0 0 1>}
 Pose_51 {type=Curve, loc=<0 0 0>, rot=<0 0 0 1>}
 Pose_52 {type=Curve, loc=<0 0 0>, rot=<0 0 0 1>}
 Pose_53 {type=Curve, loc=<0 0 0>, rot=<0 0 0 1>}
 Pose_54 {type=Curve, loc=<0 0 0>, rot=<0 0 0 1>}
 Pose_55 {type=Curve, loc=<0 0 0>, rot=<0 0 0 1>}
 Pose_56 {type=Curve, loc=<0 0 0>, rot=<0 0 0 1>}
 Pose_57 {type=Curve, loc=<0 0 0>, rot=<0 0 0 1>}
 Pose_58 {type=Curve, loc=<0 0 0>, rot=<0 0 0 1>}
 Pose_59 {type=Curve, loc=<0 0 0>, rot=<0 0 0 1>}
 Pose_60 {type=Curve, loc=<0 0 0>, rot=<0 0 0 1>}
 Pose_61 {type=Curve, loc=<0 0 0>, rot=<0 0 0 1>}
 Pose_62 {type=Curve, loc=<0 0 0>, rot=<0 0 0 1>}
 Pose_63 {type=Curve, loc=<0 0 0>, rot=<0 0 0 1>}
 Pose_64 {type=Curve, loc=<0 0 0>, rot=<0 0 0 1>}
 Pose_65 {type=Curve, loc=<0 0 0>, rot=<0 0 0 1>}
 Pose_66 {type=Curve, loc=<0 0 0>, rot=<0 0 0 1>}
 Pose_67 {type=Curve, loc=<0 0 0>, rot=<0 0 0 1>}
 Pose_68 {type=Curve, loc=<0 0 0>, rot=<0 0 0 1>}
 Pose_69 {type=Curve, loc=<0 0 0>, rot=<0 0 0 1>}
 Pose_70 {type=Curve, loc=<0 0 0>, rot=<0 0 0 1>}
 Pose_71 {type=Curve, loc=<0 0 0>, rot=<0 0 0 1>}
 Pose_72 {type=Curve, loc=<0 0 0>, rot=<0 0 0 1>}
 Pose_73 {type=Curve, loc=<0 0 0>, rot=<0 0 0 1>}
 Pose_74 {type=Curve, loc=<0 0 0>, rot=<0 0 0 1>}
 Pose_75 {type=Curve, loc=<0 0 0>, rot=<0 0 0 1>}
 Pose_76 {type=Curve, loc=<0 0 0>, rot=<0 0 0 1>}
 Pose_77 {type=Curve, loc=<0 0 0>, rot=<0 0 0 1>}
 Pose_78 {type=Curve, loc=<0 0 0>, rot=<0 0 0 1>}
 Pose_79 {type=Curve, loc=<0 0 0>, rot=<0 0 0 1>}
 Pose_80 {type=Curve, loc=<0 0 0>, rot=<0 0 0 1>}
 Pose_81 {type=Curve, loc=<0 0 0>, rot=<0 0 0 1>}
 Pose_82 {type=Curve, loc=<0 0 0>, rot=<0 0 0 1>}
 Pose_83 {type=Curve, loc=<0 0 0>, rot=<0 0 0 1>}
 Pose_84 {type=Curve, loc=<0 0 0>, rot=<0 0 0 1>}
 Pose_85 {type=Curve, loc=<0 0 0>, rot=<0 0 0 1>}
 Pose_86 {type=Curve, loc=<0 0 0>, rot=<0 0 0 1>}
 Pose_87 {type=Curve, loc=<0 0 0>, rot=<0 0 0 1>}
 Pose_88 {type=Curve, loc=<0 0 0>, rot=<0 0 0 1>}
 Pose_89 {type=Curve, loc=<0 0 0>, rot=<0 0 0 1>}
 Pose_90 {type=Curve, loc=<0 0 0>, rot=<0 0 0 1>}
 Pose_91 {type=Curve, loc=<0 0 0>, rot=<0 0 0 1>}
 Pose_92 {type=Curve, loc=<0 0 0>, rot=<0 0 0 1>}
 Pose_93 {type=Curve, loc=<0 0 0>, rot=<0 0 0 1>}
 Pose_94 {type=Curve, loc=<0 0 0>, rot=<0 0 0 1>}
 Pose_95 {type=Curve, loc=<0 0 0>, rot=<0 0 0 1>}
 Pose_96 {type=Curve, loc=<0 0 0>, rot=<0 0 0 1>}
 Pose_97 {type=Curve, loc=<0 0 0>, rot=<0 0 0 1>}
 Pose_98 {type=Curve, loc=<0 0 0>, rot=<0 0 0 1>}
 Pose_99 {type=Curve, loc=<0 0 0>, rot=<0 0 0 1>}
 Pose_100 {type=Curve, loc=<0 0 0>, rot=<0 0 0 1>}
 Pose_101 {type=Curve, loc=<0 0 0>, rot=<0 0 0 1>}
 Pose_102 {type=Curve, loc=<0 0 0>, rot=<0 0 0 1>}
 Pose_103 {type=Curve, loc=<0 0 0>, rot=<0 0 0 1>}
 Pose_104 {type=Curve, loc=<0 0 0>, rot=<0 0 0 1>}
 Pose_105 {type=Curve, loc=<0 0 0>, rot=<0 0 0 1>}
 Pose_106 {type=Curve, loc=<0 0 0>, rot=<0 0 0 1>}
 Pose_107 {type=Curve, loc=<0 0 0>, rot=<0 0 0 1>}
 Pose_108 {type=Curve, loc=<0 0 0>, rot=<0 0 0 1>}
 Pose_109 {type=Curve, loc=<0 0 0>, rot=<0 0 0 1>}
 Pose_110 {type=Curve, loc=<0 0 0>, rot=<0 0 0 1>}
 Pose_111 {type=Curve, loc=<0 0 0>, rot=<0 0 0 1>}
 Pose_112 {type=Curve, loc=<0 0 0>, rot=<0 0 0 1>}
 Pose_113 {type=Curve, loc=<0 0 0>, rot=<0 0 0 1>}
 Pose_114 {type=Curve, loc=<0 0 0>, rot=<0 0 0 1>}
 Pose_115 {type=Curve, loc=<0 0 0>, rot=<0 0 0 1>}
 Pose_116 {type=Curve, loc=<0 0 0>, rot=<0 0 0 1>}
 Pose_117 {type=Curve, loc=<0 0 0>, rot=<0 0 0 1>}
 Pose_118 {type=Curve, loc=<0 0 0>, rot=<0 0 0 1>}
 Pose_119 {type=Curve, loc=<0 0 0>, rot=<0 0 0 1>}
 Pose_120 {type=Curve, loc=<0 0 0>, rot=<0 0 0 1>}
 Pose_121 {type=Curve, loc=<0 0 0>, rot=<0 0 0 1>}
 Pose_122 {type=Curve, loc=<0 0 0>, rot=<0 0 0 1>}
 Pose_123 {type=Curve, loc=<0 0 0>, rot=<0 0 0 1>}
 Pose_124 {type=Curve, loc=<0 0 0>, rot=<0 0 0 1>}
 Pose_125 {type=Curve, loc=<0 0 0>, rot=<0 0 0 1>}
 Pose_126 {type=Curve, loc=<0 0 0>, rot=<0 0 0 1>}
 Pose_127 {type=Curve, loc=<0 0 0>, rot=<0 0 0 1>}
 Pose_128 {type=Curve, loc=<0 0 0>, rot=<0 0 0 1>}
 Pose_129 {type=Curve, loc=<0 0 0>, rot=<0 0 0 1>}
 Pose_130 {type=Curve, loc=<0 0 0>, rot=<0 0 0 1>}
 Pose_131 {type=Curve, loc=<0 0 0>, rot=<0 0 0 1>}
 Pose_132 {type=Curve, loc=<0 0 0>, rot=<0 0 0 1>}
 Pose_133 {type=Curve, loc=<0 0 0>, rot=<0 0 0 1>}
 Pose_134 {type=Curve, loc=<0 0 0>, rot=<0 0 0 1>}
 Pose_135 {type=Curve, loc=<0 0 0>, rot=<0 0 0 1>}
 Pose_136 {type=Curve, loc=<0 0 0>, rot=<0 0 0 1>}
 Pose_137 {type=Curve, loc=<0 0 0>, rot=<0 0 0 1>}
 Pose_138 {type=Curve, loc=<0 0 0>, rot=<0 0 0 1>}
 Pose_139 {type=Curve, loc=<0 0 0>, rot=<0 0 0 1>}
 Pose_140 {type=Curve, loc=<0 0 0>, rot=<0 0 0 1>}
 Pose_141 {type=Curve, loc=<0 0 0>, rot=<0 0 0 1>}
 Pose_142 {type=Curve, loc=<0 0 0>, rot=<0 0 0 1>}
 Pose_143 {type=Curve, loc=<0 0 0>, rot=<0 0 0 1>}
 Pose_144 {type=Curve, loc=<0 0 0>, rot=<0 0 0 1>}
 Pose_145 {type=Curve, loc=<0 0 0>, rot=<0 0 0 1>}
 Pose_146 {type=Curve, loc=<0 0 0>, rot=<0 0 0 1>}
 Pose_147 {type=Curve, loc=<0 0 0>, rot=<0 0 0 1>}
 Pose_148 {type=Curve, loc=<0 0 0>, rot=<0 0 0 1>}
 Pose_149 {type=Curve, loc=<0 0 0>, rot=<0 0 0 1>}
 Pose_150 {type=Curve, loc=<0 0 0>, rot=<0 0 0 1>}
 Pose_151 {type=Curve, loc=<0 0 0>, rot=<0 0 0 1>}
 Pose_152 {type=Curve, loc=<0 0 0>, rot=<0 0 0 1>}
 Pose_153 {type=Curve, loc=<0 0 0>, rot=<0 0 0 1>}
 Pose_154 {type=Curve, loc=<0 0 0>, rot=<0 0 0 1>}
 Pose_155 {type=Curve, loc=<0 0 0>, rot=<0 0 0 1>}
 Pose_156 {type=Curve, loc=<0 0 0>, rot=<0 0 0 1>}
 Pose_157 {type=Curve, loc=<0 0 0>, rot=<0 0 0 1>}
 Pose_158 {type=Curve, loc=<0 0 0>, rot=<0 0 0 1>}
 Pose_159 {type=Curve, loc=<0 0 0>, rot=<0 0 0 1>}
 Pose_160 {type=Curve, loc=<0 0 0>, rot=<0 0 0 1>}
 Pose_161 {type=Curve, loc=<0 0 0>, rot=<0 0 0 1>}
 Pose_162 {type=Curve, loc=<0 0 0>, rot=<0 0 0 1>}
 Pose_163 {type=Curve, loc=<0 0 0>, rot=<0 0 0 1>}
 Pose_164 {type=Curve, loc=<0 0 0>, rot=<0 0 0 1>}
 Pose_165 {type=Curve, loc=<0 0 0>, rot=<0 0 0 1>}
 Pose_166 {type=Curve, loc=<0 0 0>, rot=<0 0 0 1>}
 Pose_167 {type=Curve, loc=<0 0 0>, rot=<0 0 0 1>}
 Pose_168 {type=Curve, loc=<0 0 0>, rot=<0 0 0 1>}
 Pose_169 {type=Curve, loc=<0 0 0>, rot=<0 0 0 1>}
 Pose_170 {type=Curve, loc=<0 0 0>, rot=<0 0 0 1>}
 Pose_171 {type=Curve, loc=<0 0 0>, rot=<0 0 0 1>}
 Pose_172 {type=Curve, loc=<0 0 0>, rot=<0 0 0 1>}
 Pose_173 {type=Curve, loc=<0 0 0>, rot=<0 0 0 1>}
 Pose_174 {type=Curve, loc=<0 0 0>, rot=<0 0 0 1>}
 Pose_175 {type=Curve, loc=<0 0 0>, rot=<0 0 0 1>}
 Pose_176 {type=Curve, loc=<0 0 0>, rot=<0 0 0 1>}
 Pose_177 {type=Curve, loc=<0 0 0>, rot=<0 0 0 1>}
 Pose_178 {type=Curve, loc=<0 0 0>, rot=<0 0 0 1>}
 Pose_179 {type=Curve, loc=<0 0 0>, rot=<0 0 0 1>}
 eyeLookDownRight {type=Curve, loc=<0 0 0>, rot=<0 0 0 1>}
 eyeLookOutRight {type=Curve, loc=<0 0 0>, rot=<0 0 0 1>}
 eyeLookUpRight {type=Curve, loc=<0 0 0>, rot=<0 0 0 1>}
 eyeLookOutLeft {type=Curve, loc=<0 0 0>, rot=<0 0 0 1>}
 eyeBlinkLeft {type=Curve, loc=<0 0 0>, rot=<0 0 0 1>}
 eyeBlinkRight {type=Curve, loc=<0 0 0>, rot=<0 0 0 1>}
 eyeSquintLeft {type=Curve, loc=<0 0 0>, rot=<0 0 0 1>}
 eyeSquintRight {type=Curve, loc=<0 0 0>, rot=<0 0 0 1>}
 eyeWideLeft {type=Curve, loc=<0 0 0>, rot=<0 0 0 1>}
 eyeWideRight {type=Curve, loc=<0 0 0>, rot=<0 0 0 1>}
 jawForward {type=Curve, loc=<0 0 0>, rot=<0 0 0 1>}
 jawLeft {type=Curve, loc=<0 0 0>, rot=<0 0 0 1>}
 jawRight {type=Curve, loc=<0 0 0>, rot=<0 0 0 1>}
 jawOpen {type=Curve, loc=<0 0 0>, rot=<0 0 0 1>}
 mouthClose {type=Curve, loc=<0 0 0>, rot=<0 0 0 1>}
 mouthFunnel {type=Curve, loc=<0 0 0>, rot=<0 0 0 1>}
 mouthPucker {type=Curve, loc=<0 0 0>, rot=<0 0 0 1>}
 mouthLeft {type=Curve, loc=<0 0 0>, rot=<0 0 0 1>}
 mouthRight {type=Curve, loc=<0 0 0>, rot=<0 0 0 1>}
 mouthSmileLeft {type=Curve, loc=<0 0 0>, rot=<0 0 0 1>}
 mouthSmileRight {type=Curve, loc=<0 0 0>, rot=<0 0 0 1>}
 mouthFrownLeft {type=Curve, loc=<0 0 0>, rot=<0 0 0 1>}
 mouthFrownRight {type=Curve, loc=<0 0 0>, rot=<0 0 0 1>}
 mouthDimpleLeft {type=Curve, loc=<0 0 0>, rot=<0 0 0 1>}
 mouthDimpleRight {type=Curve, loc=<0 0 0>, rot=<0 0 0 1>}
 mouthStretchLeft {type=Curve, loc=<0 0 0>, rot=<0 0 0 1>}
 mouthStretchRight {type=Curve, loc=<0 0 0>, rot=<0 0 0 1>}
 mouthRollLower {type=Curve, loc=<0 0 0>, rot=<0 0 0 1>}
 mouthRollUpper {type=Curve, loc=<0 0 0>, rot=<0 0 0 1>}
 mouthShrugLower {type=Curve, loc=<0 0 0>, rot=<0 0 0 1>}
 mouthShrugUpper {type=Curve, loc=<0 0 0>, rot=<0 0 0 1>}
 mouthPressLeft {type=Curve, loc=<0 0 0>, rot=<0 0 0 1>}
 mouthPressRight {type=Curve, loc=<0 0 0>, rot=<0 0 0 1>}
 mouthLowerDownLeft {type=Curve, loc=<0 0 0>, rot=<0 0 0 1>}
 mouthLowerDownRight {type=Curve, loc=<0 0 0>, rot=<0 0 0 1>}
 mouthUpperUpLeft {type=Curve, loc=<0 0 0>, rot=<0 0 0 1>}
 mouthUpperUpRight {type=Curve, loc=<0 0 0>, rot=<0 0 0 1>}
 browDownLeft {type=Curve, loc=<0 0 0>, rot=<0 0 0 1>}
 browDownRight {type=Curve, loc=<0 0 0>, rot=<0 0 0 1>}
 browInnerUp {type=Curve, loc=<0 0 0>, rot=<0 0 0 1>}
 browOuterUpLeft {type=Curve, loc=<0 0 0>, rot=<0 0 0 1>}
 browOuterUpRight {type=Curve, loc=<0 0 0>, rot=<0 0 0 1>}
 cheekPuff {type=Curve, loc=<0 0 0>, rot=<0 0 0 1>}
 cheekSquintLeft {type=Curve, loc=<0 0 0>, rot=<0 0 0 1>}
 cheekSquintRight {type=Curve, loc=<0 0 0>, rot=<0 0 0 1>}
 noseSneerLeft {type=Curve, loc=<0 0 0>, rot=<0 0 0 1>}
 noseSneerRight {type=Curve, loc=<0 0 0>, rot=<0 0 0 1>}
 C_glabella_down {type=Curve, loc=<0 0 0>, rot=<0 0 0 1>}
 L_brow_squeeze_in {type=Curve, loc=<0 0 0>, rot=<0 0 0 1>}
 R_brow_squeeze_in {type=Curve, loc=<0 0 0>, rot=<0 0 0 1>}
 L_upper_lip_lower {type=Curve, loc=<0 0 0>, rot=<0 0 0 1>}
 R_upper_lip_lower {type=Curve, loc=<0 0 0>, rot=<0 0 0 1>}
 L_lower_lip_up {type=Curve, loc=<0 0 0>, rot=<0 0 0 1>}
 R_lower_lip_up {type=Curve, loc=<0 0 0>, rot=<0 0 0 1>}
 L_lip_corner_narrow {type=Curve, loc=<0 0 0>, rot=<0 0 0 1>}
 R_lip_corner_narrow {type=Curve, loc=<0 0 0>, rot=<0 0 0 1>}
 phoneme_ch {type=Curve, loc=<0 0 0>, rot=<0 0 0 1>}
 phoneme_mbp {type=Curve, loc=<0 0 0>, rot=<0 0 0 1>}
 phoneme_fv {type=Curve, loc=<0 0 0>, rot=<0 0 0 1>}
 mouth_up {type=Curve, loc=<0 0 0>, rot=<0 0 0 1>}
 mouth_down {type=Curve, loc=<0 0 0>, rot=<0 0 0 1>}
 L_cheek_up {type=Curve, loc=<0 0 0>, rot=<0 0 0 1>}
 R_cheek_up {type=Curve, loc=<0 0 0>, rot=<0 0 0 1>}
 L_upperLidDown {type=Curve, loc=<0 0 0>, rot=<0 0 0 1>}
 R_upperLidDown {type=Curve, loc=<0 0 0>, rot=<0 0 0 1>}
 jaw_up {type=Curve, loc=<0 0 0>, rot=<0 0 0 1>}
 jaw_back {type=Curve, loc=<0 0 0>, rot=<0 0 0 1>}
 tongue_forward {type=Curve, loc=<0 0 0>, rot=<0 0 0 1>}
 tongue_back {type=Curve, loc=<0 0 0>, rot=<0 0 0 1>}
 tongue_up {type=Curve, loc=<0 0 0>, rot=<0 0 0 1>}
 tongue_down {type=Curve, loc=<0 0 0>, rot=<0 0 0 1>}
 basePose {type=Curve, loc=<0 0 0>, rot=<0 0 0 1>}
 Pose_180 {type=Curve, loc=<0 0 0>, rot=<0 0 0 1>}
 Pose_181 {type=Curve, loc=<0 0 0>, rot=<0 0 0 1>}
 Pose_182 {type=Curve, loc=<0 0 0>, rot=<0 0 0 1>}
 Pose_183 {type=Curve, loc=<0 0 0>, rot=<0 0 0 1>}
 Pose_184 {type=Curve, loc=<0 0 0>, rot=<0 0 0 1>}
 Pose_185 {type=Curve, loc=<0 0 0>, rot=<0 0 0 1>}
 Pose_186 {type=Curve, loc=<0 0 0>, rot=<0 0 0 1>}
 Pose_187 {type=Curve, loc=<0 0 0>, rot=<0 0 0 1>}
 Pose_188 {type=Curve, loc=<0 0 0>, rot=<0 0 0 1>}
 Pose_189 {type=Curve, loc=<0 0 0>, rot=<0 0 0 1>}
 Pose_190 {type=Curve, loc=<0 0 0>, rot=<0 0 0 1>}
 Pose_191 {type=Curve, loc=<0 0 0>, rot=<0 0 0 1>}
 Pose_192 {type=Curve, loc=<0 0 0>, rot=<0 0 0 1>}
 Pose_193 {type=Curve, loc=<0 0 0>, rot=<0 0 0 1>}
 Pose_194 {type=Curve, loc=<0 0 0>, rot=<0 0 0 1>}
 Pose_195 {type=Curve, loc=<0 0 0>, rot=<0 0 0 1>}
 Pose_196 {type=Curve, loc=<0 0 0>, rot=<0 0 0 1>}
 Pose_197 {type=Curve, loc=<0 0 0>, rot=<0 0 0 1>}
 Pose_198 {type=Curve, loc=<0 0 0>, rot=<0 0 0 1>}
 Pose_199 {type=Curve, loc=<0 0 0>, rot=<0 0 0 1>}
 Pose_200 {type=Curve, loc=<0 0 0>, rot=<0 0 0 1>}
 Pose_201 {type=Curve, loc=<0 0 0>, rot=<0 0 0 1>}
 Pose_202 {type=Curve, loc=<0 0 0>, rot=<0 0 0 1>}
 Pose_203 {type=Curve, loc=<0 0 0>, rot=<0 0 0 1>}
 Pose_204 {type=Curve, loc=<0 0 0>, rot=<0 0 0 1>}
 Pose_205 {type=Curve, loc=<0 0 0>, rot=<0 0 0 1>}
 Pose_206 {type=Curve, loc=<0 0 0>, rot=<0 0 0 1>}
 Pose_207 {type=Curve, loc=<0 0 0>, rot=<0 0 0 1>}
 Pose_208 {type=Curve, loc=<0 0 0>, rot=<0 0 0 1>}
 Pose_209 {type=Curve, loc=<0 0 0>, rot=<0 0 0 1>}
 Pose_210 {type=Curve, loc=<0 0 0>, rot=<0 0 0 1>}
 Pose_211 {type=Curve, loc=<0 0 0>, rot=<0 0 0 1>}
 Pose_212 {type=Curve, loc=<0 0 0>, rot=<0 0 0 1>}
 Pose_213 {type=Curve, loc=<0 0 0>, rot=<0 0 0 1>}
 Pose_214 {type=Curve, loc=<0 0 0>, rot=<0 0 0 1>}
 Pose_215 {type=Curve, loc=<0 0 0>, rot=<0 0 0 1>}
 Pose_216 {type=Curve, loc=<0 0 0>, rot=<0 0 0 1>}
 Pose_217 {type=Curve, loc=<0 0 0>, rot=<0 0 0 1>}
 Pose_218 {type=Curve, loc=<0 0 0>, rot=<0 0 0 1>}
 Pose_219 {type=Curve, loc=<0 0 0>, rot=<0 0 0 1>}
 Pose_220 {type=Curve, loc=<0 0 0>, rot=<0 0 0 1>}
 Pose_221 {type=Curve, loc=<0 0 0>, rot=<0 0 0 1>}
 Pose_222 {type=Curve, loc=<0 0 0>, rot=<0 0 0 1>}
 Pose_223 {type=Curve, loc=<0 0 0>, rot=<0 0 0 1>}
 Pose_224 {type=Curve, loc=<0 0 0>, rot=<0 0 0 1>}
 Pose_225 {type=Curve, loc=<0 0 0>, rot=<0 0 0 1>}
 Pose_226 {type=Curve, loc=<0 0 0>, rot=<0 0 0 1>}
 Pose_227 {type=Curve, loc=<0 0 0>, rot=<0 0 0 1>}
 Pose_228 {type=Curve, loc=<0 0 0>, rot=<0 0 0 1>}
 Pose_229 {type=Curve, loc=<0 0 0>, rot=<0 0 0 1>}
 Pose_230 {type=Curve, loc=<0 0 0>, rot=<0 0 0 1>}
 Pose_231 {type=Curve, loc=<0 0 0>, rot=<0 0 0 1>}
 Pose_232 {type=Curve, loc=<0 0 0>, rot=<0 0 0 1>}
 Pose_233 {type=Curve, loc=<0 0 0>, rot=<0 0 0 1>}
 Pose_234 {type=Curve, loc=<0 0 0>, rot=<0 0 0 1>}
 Pose_235 {type=Curve, loc=<0 0 0>, rot=<0 0 0 1>}
 Pose_236 {type=Curve, loc=<0 0 0>, rot=<0 0 0 1>}
 Pose_237 {type=Curve, loc=<0 0 0>, rot=<0 0 0 1>}
 Pose_238 {type=Curve, loc=<0 0 0>, rot=<0 0 0 1>}
 Pose_239 {type=Curve, loc=<0 0 0>, rot=<0 0 0 1>}
 Pose_240 {type=Curve, loc=<0 0 0>, rot=<0 0 0 1>}
 Pose_241 {type=Curve, loc=<0 0 0>, rot=<0 0 0 1>}
 Pose_242 {type=Curve, loc=<0 0 0>, rot=<0 0 0 1>}
 Pose_243 {type=Curve, loc=<0 0 0>, rot=<0 0 0 1>}
 Pose_244 {type=Curve, loc=<0 0 0>, rot=<0 0 0 1>}
 Pose_245 {type=Curve, loc=<0 0 0>, rot=<0 0 0 1>}
 Pose_246 {type=Curve, loc=<0 0 0>, rot=<0 0 0 1>}
 Pose_247 {type=Curve, loc=<0 0 0>, rot=<0 0 0 1>}
 Pose_248 {type=Curve, loc=<0 0 0>, rot=<0 0 0 1>}
 Pose_249 {type=Curve, loc=<0 0 0>, rot=<0 0 0 1>}
 Pose_250 {type=Curve, loc=<0 0 0>, rot=<0 0 0 1>}
 Pose_251 {type=Curve, loc=<0 0 0>, rot=<0 0 0 1>}
 Pose_252 {type=Curve, loc=<0 0 0>, rot=<0 0 0 1>}
 Pose_253 {type=Curve, loc=<0 0 0>, rot=<0 0 0 1>}
 Pose_254 {type=Curve, loc=<0 0 0>, rot=<0 0 0 1>}
 Pose_255 {type=Curve, loc=<0 0 0>, rot=<0 0 0 1>}
 Pose_256 {type=Curve, loc=<0 0 0>, rot=<0 0 0 1>}
 Pose_257 {type=Curve, loc=<0 0 0>, rot=<0 0 0 1>}
 Pose_258 {type=Curve, loc=<0 0 0>, rot=<0 0 0 1>}
 neutral {type=Curve, loc=<0 0 0>, rot=<0 0 0 1>}
 Pose_259 {type=Curve, loc=<0 0 0>, rot=<0 0 0 1>}
 cheek_puff_low {type=Curve, loc=<0 0 0>, rot=<0 0 0 1>}
 Pose_260 {type=Curve, loc=<0 0 0>, rot=<0 0 0 1>}
 Pose_261 {type=Curve, loc=<0 0 0>, rot=<0 0 0 1>}
 Pose_262 {type=Curve, loc=<0 0 0>, rot=<0 0 0 1>}
 Pose_263 {type=Curve, loc=<0 0 0>, rot=<0 0 0 1>}
 Pose_264 {type=Curve, loc=<0 0 0>, rot=<0 0 0 1>}
 Pose_265 {type=Curve, loc=<0 0 0>, rot=<0 0 0 1>}
 Pose_266 {type=Curve, loc=<0 0 0>, rot=<0 0 0 1>}
 Pose_267 {type=Curve, loc=<0 0 0>, rot=<0 0 0 1>}
 Pose_268 {type=Curve, loc=<0 0 0>, rot=<0 0 0 1>}
 Pose_269 {type=Curve, loc=<0 0 0>, rot=<0 0 0 1>}
 Pose_270 {type=Curve, loc=<0 0 0>, rot=<0 0 0 1>}
 Pose_271 {type=Curve, loc=<0 0 0>, rot=<0 0 0 1>}
 Pose_272 {type=Curve, loc=<0 0 0>, rot=<0 0 0 1>}
 Pose_273 {type=Curve, loc=<0 0 0>, rot=<0 0 0 1>}
 Pose_274 {type=Curve, loc=<0 0 0>, rot=<0 0 0 1>}
 Pose_275 {type=Curve, loc=<0 0 0>, rot=<0 0 0 1>}
 Pose_276 {type=Curve, loc=<0 0 0>, rot=<0 0 0 1>}
 Pose_277 {type=Curve, loc=<0 0 0>, rot=<0 0 0 1>}
 Pose_278 {type=Curve, loc=<0 0 0>, rot=<0 0 0 1>}
 Pose_279 {type=Curve, loc=<0 0 0>, rot=<0 0 0 1>}
 Pose_280 {type=Curve, loc=<0 0 0>, rot=<0 0 0 1>}
 Pose_281 {type=Curve, loc=<0 0 0>, rot=<0 0 0 1>}
 Pose_282 {type=Curve, loc=<0 0 0>, rot=<0 0 0 1>}
 Pose_283 {type=Curve, loc=<0 0 0>, rot=<0 0 0 1>}
 Pose_284 {type=Curve, loc=<0 0 0>, rot=<0 0 0 1>}
 Pose_285 {type=Curve, loc=<0 0 0>, rot=<0 0 0 1>}
 Pose_286 {type=Curve, loc=<0 0 0>, rot=<0 0 0 1>}
 Pose_287 {type=Curve, loc=<0 0 0>, rot=<0 0 0 1>}
 Pose_288 {type=Curve, loc=<0 0 0>, rot=<0 0 0 1>}
 Pose_289 {type=Curve, loc=<0 0 0>, rot=<0 0 0 1>}
 Pose_290 {type=Curve, loc=<0 0 0>, rot=<0 0 0 1>}
 Pose_291 {type=Curve, loc=<0 0 0>, rot=<0 0 0 1>}
 Pose_292 {type=Curve, loc=<0 0 0>, rot=<0 0 0 1>}
 Pose_293 {type=Curve, loc=<0 0 0>, rot=<0 0 0 1>}
 Pose_294 {type=Curve, loc=<0 0 0>, rot=<0 0 0 1>}
 Pose_295 {type=Curve, loc=<0 0 0>, rot=<0 0 0 1>}
 Pose_296 {type=Curve, loc=<0 0 0>, rot=<0 0 0 1>}
 Pose_297 {type=Curve, loc=<0 0 0>, rot=<0 0 0 1>}
 Pose_298 {type=Curve, loc=<0 0 0>, rot=<0 0 0 1>}
 Pose_299 {type=Curve, loc=<0 0 0>, rot=<0 0 0 1>}
 Pose_300 {type=Curve, loc=<0 0 0>, rot=<0 0 0 1>}
 Pose_301 {type=Curve, loc=<0 0 0>, rot=<0 0 0 1>}
 Pose_302 {type=Curve, loc=<0 0 0>, rot=<0 0 0 1>}
 Pose_303 {type=Curve, loc=<0 0 0>, rot=<0 0 0 1>}
 Pose_304 {type=Curve, loc=<0 0 0>, rot=<0 0 0 1>}
 Pose_305 {type=Curve, loc=<0 0 0>, rot=<0 0 0 1>}
 Pose_306 {type=Curve, loc=<0 0 0>, rot=<0 0 0 1>}
 Pose_307 {type=Curve, loc=<0 0 0>, rot=<0 0 0 1>}
 Pose_308 {type=Curve, loc=<0 0 0>, rot=<0 0 0 1>}
 Pose_309 {type=Curve, loc=<0 0 0>, rot=<0 0 0 1>}
 Pose_310 {type=Curve, loc=<0 0 0>, rot=<0 0 0 1>}
 Pose_311 {type=Curve, loc=<0 0 0>, rot=<0 0 0 1>}
 Pose_312 {type=Curve, loc=<0 0 0>, rot=<0 0 0 1>}
 Pose_313 {type=Curve, loc=<0 0 0>, rot=<0 0 0 1>}
 Pose_314 {type=Curve, loc=<0 0 0>, rot=<0 0 0 1>}
 Pose_315 {type=Curve, loc=<0 0 0>, rot=<0 0 0 1>}
 Pose_316 {type=Curve, loc=<0 0 0>, rot=<0 0 0 1>}
 Pose_317 {type=Curve, loc=<0 0 0>, rot=<0 0 0 1>}
 Pose_318 {type=Curve, loc=<0 0 0>, rot=<0 0 0 1>}
 Pose_319 {type=Curve, loc=<0 0 0>, rot=<0 0 0 1>}
 Pose_320 {type=Curve, loc=<0 0 0>, rot=<0 0 0 1>}
 Pose_321 {type=Curve, loc=<0 0 0>, rot=<0 0 0 1>}
 Desert_Island_Jonesy_Mouth {type=Curve, loc=<0 0 0>, rot=<0 0 0 1>}
 Pose_322 {type=Curve, loc=<0 0 0>, rot=<0 0 0 1>}
 HT_L1_TransitionCurve {type=Curve, loc=<0 0 0>, rot=<0 0 0 1>}
 Hair_CapShape {type=Curve, loc=<0 0 0>, rot=<0 0 0 1>}
 Hair_CapShape1 {type=Curve, loc=<0 0 0>, rot=<0 0 0 1>}
 jekyll {type=Curve, loc=<0 0 0>, rot=<0 0 0 1>}
 JH_TransformCurve {type=Curve, loc=<0 0 0>, rot=<0 0 0 1>}
 eyes_LOD0 {type=Curve, loc=<0 0 0>, rot=<0 0 0 1>}
 MaterialCurveA {type=Curve, loc=<0 0 0>, rot=<0 0 0 1>}
 MaterialCurveB {type=Curve, loc=<0 0 0>, rot=<0 0 0 1>}
 MaterialCurveC {type=Curve, loc=<0 0 0>, rot=<0 0 0 1>}
 MaterialCurveD {type=Curve, loc=<0 0 0>, rot=<0 0 0 1>}
 MaterialCurveE {type=Curve, loc=<0 0 0>, rot=<0 0 0 1>}
]
